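# Physics asset for the BitterFlower_Sharp mesh in the ToyFoxGameplay content, giving it a minimal physics setup.
# It defines a single body on the handle bone, with no collision shapes listed.
# It has no constraints, so there is no jointed ragdoll chain and no other bone has its own body.
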PhysicsAsset /ToyFoxGameplay/Assets/Meshes/BitterFlower/BitterFlower_Sharp_Physics
bodies {handle []}
constraints []
